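# Translucent unlit particle material, named as the inner glow for a glider contrail engine, usable on sprite and mesh particles.
# A SphereMask centered in UV space forms a radial glow with parameterized radius and hardness, multiplied by particle alpha.
# DepthFade softens the result by a parameterized distance.
# A feature level switch skips that fade on ES2 and ES3.1, using the plain masked alpha instead.
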Material /Figment_S09/Athena/Effects/Materials/M_GliderContrail_Engine_InnerGlow
graph {
 properties {
  bUsedWithMeshParticles=true
  bUsedWithParticleSprites=true
  blendMode=BLEND_Translucent
  pixelDepthOffsetMode=PDOM_Legacy
  shadingModel=MSM_Unlit
  usedShadingModels=Unlit
 }
 nodes [
  n1 = Multiply(
   A=ParticleColor().A
   B=SphereMask(
    A=TextureCoordinate()
    B=Constant2Vector() {g=0.5, r=0.5}
    Radius=ScalarParameter() {defaultValue=0.5, parameterName=GlowRadius}
    Hardness=ScalarParameter() {parameterName=GlowHardness}
   )
  )
  n2 = DepthFade(Opacity=@n1, FadeDistance=ScalarParameter() {defaultValue=25, parameterName=FadeDistance})
  FeatureLevelSwitch(Default=@n2, ES2=@n1, ES3_1=@n1, SM4_REMOVED=@n2)
 ]
}
